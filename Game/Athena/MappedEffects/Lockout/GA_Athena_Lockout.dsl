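; Gameplay ability for Athena that locks out the player while active by holstering their weapon under the "LockoutHolster" reason.
; Activation holsters the avatar pawn's weapon and starts a timer of LockoutFailsafeTime that ends the ability if nothing else has.
; Ability end and failed passive activation both unholster the weapon and clear that timer.
; All holster and timer steps run only when the avatar is a valid Fort player pawn.
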(event Custom|Failsafe
  ;[1 1 1] Lockout Failsafe time
  (Ability|EndAbility))

(event Ability|EventActivateAbility
  ;[1 1 1] Holster and Unholster to lock out the player during the ability
  (bind _outfortplayerpawn (CallFunction|GetAvatarasValidFortPlayerPawn)
    (:ValidFortPlayerPawn
      (Weapon|SetHolsterWeapon _outfortplayerpawn 0 true false "LockoutHolster")
      ;[1 1 1] Lockout Failsafe time
      (bind _returnvalue (Utilities|Time|SetTimerbyEvent (delegate Custom|Failsafe) (Variables|Default|GetLockoutFailsafeTime)))
      (Variables|Default|SetTimer_LockoutFailsafe _returnvalue))))

(event Ability|EventOnEndAbility
  (params (bWasCancelled :type Boolean))
  ;[1 1 1] Holster and Unholster to lock out the player during the ability
  (bind _outfortplayerpawn (CallFunction|GetAvatarasValidFortPlayerPawn)
    (:ValidFortPlayerPawn
      (Weapon|SetHolsterWeapon _outfortplayerpawn 0 false true "LockoutHolster")
      ;[1 1 1] Lockout Failsafe time
      (Utilities|Time|ClearandInvalidateTimerbyHandle (Variables|Default|GetTimer_LockoutFailsafe)))))

(event Ability|EventFailedToActivatePassiveAbility
  (params (ActorInfo :type Gameplay%20Ability%20Actor%20Info%20Structure))
  ;[1 1 1] Holster and Unholster to lock out the player during the ability
  (bind _outfortplayerpawn (CallFunction|GetAvatarasValidFortPlayerPawn)
    (:ValidFortPlayerPawn
      (Weapon|SetHolsterWeapon _outfortplayerpawn 0 false true "LockoutHolster")
      ;[1 1 1] Lockout Failsafe time
      (Utilities|Time|ClearandInvalidateTimerbyHandle (Variables|Default|GetTimer_LockoutFailsafe)))))
